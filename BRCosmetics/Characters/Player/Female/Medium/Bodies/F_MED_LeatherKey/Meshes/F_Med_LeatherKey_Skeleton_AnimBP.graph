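# Post-copy AnimBlueprint for the F_MED_LeatherKey female body cosmetic, adding secondary motion to the pose copied from the owner's mesh.
# Control Rig feeds Anim Dynamics on bandana, leg strip and tags, with tags easing in after customization loads.
# A thigh/calf constraint and right-knee-driven additive transform adjust the left knee pad; a RigidBody node sits unconnected.
# The event graph hides a material element when the pawn has the JumpBoots tag.
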AnimBlueprint /BRCosmetics/Characters/Player/Female/Medium/Bodies/F_MED_LeatherKey/Meshes/F_Med_LeatherKey_Skeleton_AnimBP
target_skeleton=&/F_Med_LeatherKey_Skeleton
graphs [
 AnimGraph {nodes [
   RigidBody() {
    bClampLinearTranslationLimitToRefPose=true
    bOverrideWorldGravity=true
    overridePhysicsAsset=/Game/Characters/Player/Female/Medium/Bodies/F_Med_Soldier_01/Meshes/F_MED_Soldier_01_Dyn_legs
    simulationSpace=BaseBoneSpace
    alphaBoolBlend {blendOption=Linear}
    baseBoneRef {boneName=spine_01}
    componentAppliedLinearAccClamp=<2000 2000 2000>
    componentLinearAccScale=<0.5 0.5 0.5>
    componentLinearVelScale=<0.5 0.5 0.5>
    overrideWorldGravity {z=-980}
   }
   n1 = Control Rig(Source=Copy Pose From Mesh(SourceMeshComponent=Get MeshToCopy() -> (MeshToCopy:Skeletal Mesh Component Object Reference))) {
    alphaBoolBlend {blendOption=Linear}
    controlRigAssetReference {blueprintRigClass=&/F_MED_LeatherKey_CtrlRig.F_MED_LeatherKey_CtrlRig_C}
   }
   n2 = "Transform (Modify) Bone - Bone: dyn_tags"(ComponentPose=Local To Component(LocalPose=@n1), Alpha=1, Translation="0,0,0", Rotation="0,-6.5,0", Scale="1,1,1") {
    rotationMode=BMM_Additive
    rotationSpace=BCS_ParentBoneSpace
    alphaBoolBlend {blendOption=Linear}
    boneToModify {boneName=dyn_tags}
    rotation {yaw=-6.5}
   }
   n3 = Anim Dynamics - Bone: dyn_arm_r_bandana(ComponentPose=@n2, Alpha=0.75, ExternalForce="0,0,0") {
    alpha=0.9
    angularSpringConstant=1500
    bAngularSpring=true
    bGravityOverrideInSimSpace=true
    numSolverIterationsPostUpdate=4
    numSolverIterationsPreUpdate=16
    simulationSpace=Actor
    alphaBoolBlend {blendOption=Linear}
    boundBone {boneName=dyn_arm_r_bandana}
    physicsBodyDefinitions [{
      sphereCollisionRadius=0
      boundBone {boneName=dyn_arm_r_bandana}
      boxExtents=<15 5 5>
      constraintSetup {angularTargetAxis=AxisZ, twistAxis=AxisY, angularLimitsMax=<15 0 15>, angularLimitsMin=<-15 0 -15>, angularTarget=<0 0 1>}
      localJointOffset=<0 0 0>
     }]
   }
   n4 = Anim Dynamics - Bone: dyn_strip_leg_r(ComponentPose=@n3, Alpha=0.75, ExternalForce="0,0,0") {
    alpha=0.9
    angularSpringConstant=1500
    bAngularSpring=true
    bGravityOverrideInSimSpace=true
    numSolverIterationsPostUpdate=4
    numSolverIterationsPreUpdate=16
    simulationSpace=Actor
    alphaBoolBlend {blendOption=Linear}
    boundBone {boneName=dyn_strip_leg_r}
    physicsBodyDefinitions [{
      sphereCollisionRadius=0
      boundBone {boneName=dyn_strip_leg_r}
      boxExtents=<15 5 5>
      constraintSetup {angularTargetAxis=AxisZ, twistAxis=AxisY, angularLimitsMax=<15 0 15>, angularLimitsMin=<-15 0 -15>, angularTarget=<0 0 1>}
      localJointOffset=<0 0 0>
     }]
   }
   n5 = Anim Dynamics - Bone: dyn_tags(ComponentPose=@n4, Alpha=Get ease_in_animdyn() -> (ease_in_animdyn:double), ExternalForce="0,0,0") {
    angularBiasOverride=1
    angularDampingOverride=0.98
    bGravityOverrideInSimSpace=true
    lODThreshold=1
    numSolverIterationsPostUpdate=4
    numSolverIterationsPreUpdate=16
    alphaBoolBlend {blendOption=Linear}
    alphaScaleBiasClamp {bInterpResult=true, interpSpeedIncreasing=0.01, scale=0.3}
    boundBone {boneName=dyn_tags}
    physicsBodyDefinitions [{
      sphereCollisionRadius=0
      boundBone {boneName=dyn_tags}
      boxExtents=<5 2 10>
      constraintSetup {twistAxis=AxisZ, angularLimitsMax=<25 10 0>, angularLimitsMin=<0 -10 0>, angularTarget=<0 0 0>}
      localJointOffset=<0 1 -1>
     }]
    relativeSpaceBone {boneName=dyn_tags}
   }
   n6 = "Constraint - dyn_pad_leg_l (PP)"!(ComponentPose=@n5, Alpha=1, ConstraintWeights_0=0.5, ConstraintWeights_1:float=0.5) -> (Pose) {
    alphaBoolBlend {blendOption=Linear}
    boneToModify {boneName=dyn_pad_leg_l}
    constraintSetup [{targetBone {boneName=thigh_l}}, {targetBone {boneName=calf_l}}]
    constraintWeights [1,1]
   }
   n7 = "Transform (Modify) Bone - Bone: dyn_pad_leg_l"(ComponentPose=@n6, Alpha=Get Knee_Pad_Rot() -> (Knee_Pad_Rot:double), Translation="0,-1,0", Rotation="0,20,0", Scale="1,1,1.05") {
    rotationMode=BMM_Additive
    rotationSpace=BCS_BoneSpace
    scaleMode=BMM_Additive
    scaleSpace=BCS_BoneSpace
    translationMode=BMM_Additive
    translationSpace=BCS_BoneSpace
    alphaBoolBlend {blendOption=Linear}
    boneToModify {boneName=dyn_pad_leg_l}
    rotation {yaw=20}
    scale {z=1.05}
    translation {y=-1}
   }
   Output Pose(Result=Component To Local(ComponentPose=@n7))
  ]}
 EventGraph {nodes [
   Event OnCharacterCustomizationLoaded()
   > Set ease_in_animdyn(ease_in_animdyn:double=1) -> (Output_Get:double)
   Event BlueprintUpdateAnimation()
   > n1 = Sequence()
   > n2 = Cast To FortPlayerPawn(Object=GetOwningActor())
   > Is Valid(InputObject=@n2)
   > n3 = HasMatchingGameplayTag!(execute:Exec, self:Object Reference=@n2, TagToCheck:Gameplay Tag Structure=(TagName="Item.Gadget.JumpBoots.AnimTracker")) -> (then:Exec, ReturnValue:bool) {is=title K2Node_Message}
   > Branch(Condition=@n3)
   > n4 = Get Dynamic Material Instance(SkeletalMeshComponent=GetPartSkeletalMeshComponent())
   > SetScalarParameterValue(self=@n4, ParameterName=Hide Element 03, Value=1)
   Set SkelMesh(execute=@n1.then_1, SkelMesh=GetPartSkeletalMeshComponent())
   > Set Knee_Pad_Rot(Knee_Pad_Rot:double=GetAlphaFromReferenceTransformDelta(SkeletalMeshComponent=Get SkelMesh(), BoneName=calf_r, BaseName=thigh_r, InputRange=(X=0,Y=120), TwistAxis="0,0,1")) -> (Output_Get:double)
   Event BlueprintInitializeAnimation()
   > Set MeshToCopy(MeshToCopy:Skeletal Mesh Component Object Reference=GetOwnerSkeletalMeshComponent()) -> (Output_Get:Skeletal Mesh Component Object Reference)
   TryGetPawnOwner()
  ]}
]
variables [
 ease_in_animdyn {type=number}
 Knee_Pad_Rot {type=number}
 MeshToCopy {type=SkeletalMeshComponent}
 SkelMesh {type=SkeletalMeshComponent}
]
